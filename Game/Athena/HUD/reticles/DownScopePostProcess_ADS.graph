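# Post-process material for the thermal scope aim-down-sights view, driven by MPC_ThermalGlobal scale and fadeout.
# Scene UVs get animated pixelation, zoom and noise distortion, with simple scaling on mobile.
# Custom depth, stencil and distance select heat-colored areas, using a normal-based gradient on deferred or luminance on mobile/forward, over a tinted scene.
# Mask textures add the scope border, reticle graphics, a scrolling center-depth range readout and flickering random digits.
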Material /Game/Athena/HUD/reticles/DownScopePostProcess_ADS
graph {
 properties {
  bEnableResponsiveAA=true
  blendMode=BLEND_Translucent
  blendableLocation=BL_SceneColorAfterDOF
  materialDomain=MD_PostProcess
  pixelDepthOffsetMode=PDOM_Legacy
  usedShadingModels=Default Lit
 }
 nodes [
  n1 = ViewProperty() {property=MEVP_ViewSize}
  n2 = Divide(A=ComponentMask(None=@n1.Property) {g=true}, B=ComponentMask(None=@n1.Property) {r=true})
  n3 = Constant() {r=1}
  n4 = MaterialFunctionCall(
   UVs=ScreenPosition().ViewportUV
   Texture Scale=Multiply(A=AppendVector(A=@n2, B=@n3), B=ScalarParameter() {defaultValue=1, parameterName=ADS_MaskSize})
  ) {materialFunction=&ScaleUVsByCenter}
  n5 = TextureSample(UVs=@n4.UVs) {samplerType=SAMPLERTYPE_Masks, texture=/Game/Athena/HUD/reticles/Weapons/T_ADS_Mask}
  n6 = Multiply()
  Add(A=@n5.G, B=@n6)
  n7 = If(A=@n5.G, B=Constant() {r=0.129}, "A > B"=Constant() {r=1}, "A == B"=@n5.G, "A < B"=@n5.G)
  n8 = OneMinus(None=SphereMask(A=TextureCoordinate(), B=Constant2Vector() {g=0.5, r=0.5}, Radius=Constant() {r=0.05}))
  n9 = Add(A=Saturate(None=Add(A=@n7, B=@n8))) {constB=0.85}
  n10 = CollectionParameter() {collection=/Game/Athena/Items/Weapons/Abilities/Thermal/MPC_ThermalGlobal, parameterName=ThermalScale}
  n11 = LinearInterpolate(A=@n10, Alpha=ScalarParameter() {group=Debug, parameterName=Debug})
  n12 = Multiply(A=@n11, B=Constant() {r=1.5})
  n13 = Constant() {r=1.7}
  n14 = OneMinus(None=Clamp(None=Multiply(A=@n12, B=@n13) {constB=1.7}))
  n15 = Constant() {r=4}
  n16 = Clamp(None=Multiply(A=@n12, B=Constant() {r=2}) {constB=1.7})
  n17 = AppendVector(A=Add(A=Multiply(A=@n14, B=@n15) {constB=4}, B=Constant() {r=1}), B=Multiply(A=@n16, B=@n16))
  n18 = TextureCoordinate()
  n19 = Multiply(A=ViewProperty() {property=MEVP_ViewSize}.Property, B=LinearInterpolate(A=Constant() {r=0.05}, Alpha=Multiply(A=@n11, B=@n11) {constB=2}))
  n20 = Floor(None=Multiply(A=@n18, B=@n19))
  n21 = Constant() {r=100}
  n22 = Constant() {r=-100}
  n23 = MaterialFunctionCall!(RotateDialMask, Normal WS) {materialFunction=/Game/Packages/Fortress_SharedMaterials/Base_Material_Functions/AttachScope}
  n24 = Distance(A=VertexInterpolator(VS=ComponentMask(None=@n23.Post process uvs 0-1 uvs) {g=true, r=true}).PS, B=Constant() {r=0.5})
  n25 = Clamp(None=Min(A=LinearInterpolate(A=@n21, B=@n22, Alpha=@n24), B=ScalarParameter() {defaultValue=1, parameterName=BlendAmount}))
  n26 = ShadingPathSwitch(
   Default=LinearInterpolate(A=@n18, B=Divide(A=@n20, B=@n19), Alpha=@n25)
   Mobile=MaterialFunctionCall(UVs=TextureCoordinate(), Texture Scale=ScalarParameter() {defaultValue=1, parameterName=Mobile_ScopeMaskScale}) {materialFunction=&ScaleUVsByCenter}.UVs
  ) {desc=Migrated from FeatureLevelSwitch}
  n27 = TextureSample(UVs=Multiply(A=Time(), B=ScalarParameter() {defaultValue=24.535, group=Noise, parameterName=RandomNoiseFrequency})) {samplerType=SAMPLERTYPE_Masks, texture=/Game/Athena/HUD/reticles/Weapons/T_ADS_Noise}
  n28 = ScalarParameter() {defaultValue=0.8, group=Noise, parameterName=RandomNoiseBias}
  n29 = Saturate(None=Multiply(A=Subtract(A=@n27.G, B=@n28), B=ScalarParameter() {defaultValue=8, group=Noise, parameterName=RandomNoiseMult}))
  n30 = ScreenPosition()
  n31 = ScalarParameter() {defaultValue=4070.1392, group=Noise, parameterName=NoiseScale}
  n32 = Constant()
  n33 = ScalarParameter() {group=Noise, parameterName=DistortionPan}
  n34 = TextureSample(UVs=Panner(Coordinate=Multiply(A=@n30.ViewportUV, B=@n31), Speed=AppendVector(A=@n32, B=@n33))) {samplerType=SAMPLERTYPE_Masks, texture=&T_ADS_Noise}
  n35 = AppendVector(A=ScalarParameter() {group=Noise, parameterName=Distortion1}, B=ScalarParameter() {group=Noise, parameterName=Distortion2})
  n36 = Add(
   A=LinearInterpolate(
    A=@n26
    B=MaterialFunctionCall(UVs=@n26, Texture Scale=@n17) {materialFunction=&ScaleUVsByCenter}.UVs
    Alpha=Multiply(A=@n14, B=@n25)
   )
   B=Multiply(A=@n29, B=Multiply(A=@n34.G, B=@n35))
  )
  n37 = ComponentMask(None=SceneTexture(UVs=@n36) {bFiltered=true, sceneTextureId=PPI_PostProcessInput0}.Color) {b=true, g=true, r=true}
  n38 = Multiply(
   A=VectorParameter() {parameterName=TInt, defaultValue=<rgba 0.3011 0.4005 0.8958 1>}.RGB
   B=Desaturation(None=@n37, Fraction=ScalarParameter() {parameterName=DesatSceneColor})
  )
  n39 = VectorParameter() {parameterName=TIntHeat2, defaultValue=<rgba 0.3011 0.4005 0.8958 1>}
  n40 = VectorParameter() {parameterName=TIntHeat, defaultValue=<rgba 0.3011 0.4005 0.8958 1>}
  n41 = Desaturation(None=@n37)
  n42 = ScalarParameter() {defaultValue=0.5, parameterName=HeatGradientPower}
  n43 = Saturate(None=Multiply(A=Power(Base=@n41, Exp=@n42), B=ScalarParameter() {defaultValue=0.5, parameterName=HeatGradientMult}))
  n44 = ComponentMask(None=SceneTexture(UVs=@n36) {bFiltered=true, sceneTextureId=PPI_WorldNormal}.Color) {b=true, g=true, r=true}
  n45 = CameraVectorWS()
  n46 = Multiply(
   A=Power(Base=DotProduct(A=@n44, B=@n45), Exp=ScalarParameter() {defaultValue=3, parameterName=FresnelGradient})
   B=ScalarParameter() {defaultValue=1, parameterName=FresnelMult}
  )
  n47 = TextureSample(UVs=ShadingPathSwitch(Default=@n46, Deferred=@n46, Forward=@n43, Mobile=@n43)) {texture=/Game/Athena/HUD/reticles/Weapons/T_ADS_Gradient}
  n48 = Power(Base=@n47.RGB, Exp=ScalarParameter() {defaultValue=3, parameterName=BrightnessPower})
  n49 = ScalarParameter() {defaultValue=3, parameterName=BrightnessHeat}
  n50 = Add(A=Multiply(A=LinearInterpolate(A=@n39.RGB, B=@n40.RGB, Alpha=@n47.R), B=Multiply(A=@n48, B=@n49)), B=@n41)
  n51 = ComponentMask(None=SceneTexture(UVs=@n36) {bFiltered=true, sceneTextureId=PPI_SceneDepth}.Color) {r=true}
  n52 = Divide(
   A=Add(A=@n51, B=ScalarParameter() {defaultValue=-100000, parameterName=DistanceOffset})
   B=ScalarParameter() {defaultValue=100000, parameterName=DistanceFade}
  )
  n53 = LinearInterpolate(
   A=CollectionParameter() {collection=&MPC_ThermalGlobal, parameterName=ThermalScale}
   Alpha=ScalarParameter() {group=Debug, parameterName=Debug}
  )
  n54 = ScalarParameter() {defaultValue=5, parameterName=Test}
  n55 = SceneTexture(UVs=@n36) {bFiltered=true, sceneTextureId=PPI_CustomDepth}
  n56 = Clamp(None=Subtract(A=Add(A=@n54, B=@n51) {constB=0.5}, B=ComponentMask(None=@n55.Color) {r=true}))
  n57 = MaterialFunctionCall!(ID=Constant() {r=2}) {materialFunction=/Game/Packages/Fortress_SharedMaterials/Stencil_ID_Definitions/StencilMaskCompare}
  n58 = Multiply(A=OneMinus(None=Saturate(None=@n52)), B=Min(A=Multiply(A=@n53, B=@n56), B=OneMinus(None=@n57.Result)))
  n59 = ScalarParameter() {defaultValue=0.1, group=Noise, parameterName=PanOverlayBrightness}
  n60 = Multiply(A=Multiply(A=ScalarParameter() {defaultValue=1, group=Noise, parameterName=PanOverlayDarkNess}, B=@n34.G), B=@n34.R)
  n61 = Multiply(
   A=Add(A=LinearInterpolate(A=@n38, B=@n50, Alpha=@n58), B=Multiply(A=@n59, B=@n34.G))
   B=LinearInterpolate(B=OneMinus(None=@n60), Alpha=@n29) {constA=1}
  )
  n62 = OneMinus(None=MaterialFunctionCall(ID=Constant() {r=123}) {materialFunction=&StencilMaskCompare}.Result)
  n63 = ScalarParameter() {group=Debug, parameterName=Debug Blackness}
  n64 = Power(Base=OneMinus(None=LinearInterpolate(A=@n10, Alpha=@n63)))
  n65 = Multiply(A=Multiply(A=LinearInterpolate(A=@n37, B=@n61, Alpha=@n62), B=OneMinus(None=@n64)), B=@n5.B)
  n66 = Subtract(A=VectorParameter() {parameterName=BorderColorOuterEdge, defaultValue=<rgba 1 0.2917 0.2917 1>}.RGB, B=@n6)
  n67 = VectorParameter() {parameterName=BorderColor, defaultValue=<rgba 1 1 1 1>}
  n68 = AppendVector(
   A=ScalarParameter() {defaultValue=1, parameterName=MaskSizeInnerOverlay_X}
   B=ScalarParameter() {defaultValue=1, parameterName=MaskSizeInnerOverlay_Y}
  )
  n69 = Multiply(
   A=Power(Base=@n5.G, Exp=ScalarParameter() {defaultValue=2, parameterName=ContrastADSmask})
   B=LinearInterpolate(
    A=LinearInterpolate(A=@n66, B=@n67.RGB, Alpha=@n5.G)
    B=VectorParameter() {parameterName=BorderColorCore, defaultValue=<rgba 1 1 1 1>}.RGB
    Alpha=MaterialFunctionCall(UVs=@n4.UVs, Texture Scale=@n68) {materialFunction=&ScaleUVsByCenter}.0-1 mask
   )
  )
  n70 = ScalarParameter() {defaultValue=2, parameterName=ContrastADSGraphicDesign}
  n71 = ScalarParameter() {defaultValue=2, parameterName=ContrastADSGraphicDesign2}
  n72 = Time()
  n73 = ScalarParameter() {defaultValue=1, parameterName=RandomNumberSpeed}
  n74 = Multiply(A=Floor(None=Multiply(A=@n72, B=@n73)), B=ScalarParameter() {defaultValue=24.535, parameterName=Random})
  n75 = Frac(None=@n74)
  n76 = LinearInterpolate(
   A=VectorParameter() {parameterName=GrapicDesignColor, defaultValue=<rgba 1 0.5817 0 1>}.RGB
   B=VectorParameter() {parameterName=GrapicDesignColor2, defaultValue=<rgba 1 0.0859 0 1>}.RGB
   Alpha=Power(Base=@n5.R, Exp=LinearInterpolate(A=@n70, B=@n71, Alpha=@n75))
  )
  n77 = ScalarParameter() {parameterName=Offset_ADS_Distance_X}
  n78 = ScalarParameter() {parameterName=Offset_ADS_Distance_Y}
  n79 = MaterialFunctionCall(
   UVs=Add(A=@n4.UVs, B=AppendVector(A=@n77, B=@n78))
   Texture Scale=AppendVector(
    A=ScalarParameter() {defaultValue=1, parameterName=Scale_ADS_Distance_X}
    B=ScalarParameter() {defaultValue=1, parameterName=Scale_ADS_Distance_Y}
   )
  ) {materialFunction=&ScaleUVsByCenter}
  n80 = TextureSample(UVs=@n79.UVs) {samplerType=SAMPLERTYPE_Masks, texture=/Game/Athena/HUD/reticles/Weapons/T_ADS_Mask_2}
  n81 = SceneTexture(UVs=Constant2Vector() {g=0.5, r=0.5}) {bFiltered=true, sceneTextureId=PPI_SceneDepth}
  n82 = Divide(A=Divide(A=ComponentMask(None=@n81.Color) {r=true}) {constB=100}) {constB=5}
  n83 = Multiply(A=Round(None=Subtract(A=@n82) {constB=0.5})) {constB=5}
  n84 = AppendVector(
   A=ScalarParameter() {defaultValue=0.5, group=Numbers, parameterName=u_NumberPaddingScale}
   B=ScalarParameter() {defaultValue=0.3, group=Numbers, parameterName=v_NumberPaddingScale}
  )
  n85 = AppendVector(A=Constant() {r=1}, B=ScalarParameter() {defaultValue=1.5, parameterName=Scale ADS Y})
  n86 = Multiply(A=Frac(None=@n82)) {constB=-1}
  n87 = Add(
   A=Multiply(
    A=MaterialFunctionCall(UVs=@n79.UVs, Texture Scale=@n85) {materialFunction=&ScaleUVsByCenter}.UVs
    B=Constant2Vector() {g=3, r=1}
   )
   B=AppendVector(A=Constant(), B=Add(A=@n86) {constB=0.5})
  )
  n88 = MaterialFunctionCall!(Number=Add(A=@n83) {constB=5}, UVPadding=@n84, UVs=@n87) {materialFunction=&/MF_ADS_Distance}
  n89 = Constant2Vector() {g=-1}
  n90 = Add(A=@n87, B=@n89)
  n91 = MaterialFunctionCall(Number=@n83, UVPadding=@n84, UVs=@n90) {materialFunction=&MF_ADS_Distance}
  n92 = Add(A=@n83) {constB=-5}
  n93 = Add(A=@n90, B=@n89)
  n94 = Multiply(
   A=@n80.G
   B=Add(
    A=Add(A=@n88.NumbersGreyscale, B=@n91.NumbersGreyscale)
    B=MaterialFunctionCall(Number=@n92, UVPadding=@n84, UVs=@n93) {materialFunction=&MF_ADS_Distance}.NumbersGreyscale
   )
  )
  n95 = MaterialFunctionCall(UVs=@n87, Texture Scale=Constant2Vector() {g=3, r=2}) {materialFunction=&ScaleUVsByCenter}
  n96 = Add(
   A=Add(A=@n80.R, B=Multiply(A=@n80.R, B=@n94))
   B=Multiply(A=@n80.G, B=TextureSample(UVs=@n95.UVs) {samplerType=SAMPLERTYPE_Masks, texture=&T_ADS_Mask_2}.B)
  )
  n97 = ScalarParameter() {defaultValue=2, parameterName=BrightnessHighlightPower}
  n98 = Add(A=Add(A=@n80.G, B=@n94), B=Multiply(A=Power(Base=@n96, Exp=@n97), B=ScalarParameter() {defaultValue=2, parameterName=BrightnessHighlight}))
  n99 = VectorParameter() {parameterName=Color_ADS_Distance, defaultValue=<rgba 0.224 0.8385 0.0459 1>}
  n100 = Add(A=Add(A=Add(A=@n65, B=@n69), B=Multiply(A=@n5.R, B=@n76)), B=Multiply(A=@n79.0-1 mask, B=Multiply(A=@n98, B=@n99.RGB)))
  n101 = Add(
   A=@n4.UVs
   B=AppendVector(A=ScalarParameter() {parameterName=Offset_ADS3_Distance_X}, B=ScalarParameter() {parameterName=Offset_ADS3_Distance_Y})
  )
  n102 = ScalarParameter() {defaultValue=1, parameterName=Scale_ADS3_Distance_Y}
  n103 = AppendVector(A=ScalarParameter() {defaultValue=1, parameterName=Scale_ADS3_Distance_X}, B=@n102)
  n104 = ScalarParameter() {defaultValue=0.5, parameterName=NumbersOfRandomDigits}
  n105 = Constant() {r=1}
  n106 = Multiply(A=Divide(A=Constant() {r=1}, B=ScalarParameter() {defaultValue=7, parameterName=NumbersOfRandomOffset}) {constB=6}, B=Floor(None=@n74))
  n107 = Constant()
  n108 = Add(
   A=Multiply(
    A=MaterialFunctionCall(UVs=@n101, Texture Scale=@n103) {materialFunction=&ScaleUVsByCenter}.UVs
    B=AppendVector(A=@n104, B=@n105)
   )
   B=Frac(None=AppendVector(A=@n106, B=@n107))
  )
  n109 = ScalarParameter() {defaultValue=1, parameterName=Scale_ADS3_Distance_X_Mask}
  n110 = Multiply(
   A=Multiply(
    A=VectorParameter() {parameterName=ColorRandomnumbers, defaultValue=<rgba 0.6458 0 0 1>}.RGB
    B=TextureSample(UVs=@n108) {samplerType=SAMPLERTYPE_Masks, texture=/Game/Athena/HUD/reticles/Weapons/T_ADS_Mask_3}.RGB
   )
   B=MaterialFunctionCall(UVs=@n101, Texture Scale=AppendVector(A=@n109, B=@n102)) {materialFunction=&ScaleUVsByCenter}.0-1 mask
  )
  LinearInterpolate(
   A=@n37
   B=Multiply(A=Saturate(None=@n9), B=Add(A=@n100, B=@n110))
   Alpha=CollectionParameter() {collection=&MPC_ThermalGlobal, parameterName=ThermalFadeout}
  )
  Multiply(A=@n5.G)
  ScalarParameter()
 ]
}
