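# Translucent post-process material that draws an aim-down-scope lens view over the screen, with tint, reticle and optional zoom.
# AttachScope UVs, computed per vertex, drive a hard circular lens mask, sphere-normal refraction of scene color and edge darkening.
# Reticle texture channels add colored crosshairs, followed by fake cubemap reflection and vignette, blended by BlendAmount and optional cutoff masks.
# The mobile path skips refraction, lens AO, reflection and vignette.
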Material /Game/Athena/HUD/reticles/DownScopePostProcess_Athena_Color
graph {
 properties {blendMode=BLEND_Translucent, materialDomain=MD_PostProcess, pixelDepthOffsetMode=PDOM_Legacy, usedShadingModels=Default Lit}
 nodes [
  n1 = VectorParameter() {parameterName=Lens Coloration, defaultValue=<rgba 0.9075 1.0679 1.1 0.25>}
  n2 = VertexInterpolator(VS=MaterialFunctionCall!(RotateDialMask, Normal WS) {materialFunction=/Game/Packages/Fortress_SharedMaterials/Base_Material_Functions/AttachScope}.Post process uvs 0-1 uvs)
  n3 = Constant() {r=0.5}
  n4 = Distance(A=@n2.PS, B=@n3)
  n5 = ScalarParameter() {defaultValue=0.035, parameterName=Lens AO Radius}
  n6 = LinearInterpolate(
   B=ScalarParameter() {defaultValue=0.4431, parameterName=Lens AO Value, sliderMax=1}
   Alpha=Power(Base=SphereMask(A=@n4, B=@n3, Radius=@n5) {hardnessPercent=0})
  ) {constA=1, constB=0}
  n7 = Multiply(A=@n1.RGB, B=@n6) {desc=ao}
  n8 = Constant() {r=0.5}
  n9 = Clamp(None=LinearInterpolate(A=Constant() {r=100}, B=Constant() {r=-100}, Alpha=Distance(A=@n2.PS, B=@n8)))
  n10 = TextureCoordinate()
  n11 = ScalarParameter() {defaultValue=-0.05, parameterName=RefractionAmount}
  n12 = MaterialFunctionCall(Float2=ViewSize()) {materialFunction=&BreakOutFloat2Components}
  n13 = Divide(A=@n12.G, B=@n12.G)
  n14 = TextureSample(UVs=MaterialFunctionCall(UVs=@n2.PS, Texture Scale=Constant() {r=1.15}) {materialFunction=&ScaleUVsByCenter}.UVs) {samplerType=SAMPLERTYPE_LinearColor, texture=/Game/Packages/Fortress_SharedMaterials/Textures/SphereNormalsUncomp}
  n15 = ConstantBiasScale(None=@n14.RGB) {bias=-0.5, scale=2}
  n16 = Multiply(
   A=MaterialFunctionCall(X=Multiply(A=@n11, B=@n13), Y=@n11) {materialFunction=&MakeFloat2}.Result
   B=ComponentMask(None=Multiply(A=@n4, B=@n15)) {g=true, r=true}
  )
  n17 = ShadingPathSwitch(Default=LinearInterpolate(A=@n10, B=Add(A=@n10, B=@n16), Alpha=@n9), Mobile=TextureCoordinate()) {desc=Migrated from FeatureLevelSwitch}
  n18 = ComponentMask(None=SceneTexture(UVs=@n17) {bFiltered=true, sceneTextureId=PPI_PostProcessInput0}.Color) {b=true, g=true, r=true}
  n19 = Divide(A=Subtract(A=@n17) {constB=0.5}, B=ScalarParameter() {defaultValue=1, parameterName=ZoomFactor})
  n20 = ComponentMask(None=SceneTexture(UVs=Add(A=@n19) {constB=0.5}) {bFiltered=true, sceneTextureId=PPI_PostProcessInput0}.Color) {b=true, g=true, r=true}
  n21 = NamedRerouteUsage()
  n22 = Multiply(
   A=StaticSwitchParameter(True=LinearInterpolate(A=@n18, B=@n20, Alpha=@n21), False=@n18) {parameterName=useScopeZoom}
   B=LinearInterpolate(B=ShadingPathSwitch(Default=@n7, Mobile=@n1.RGB) {desc=Migrated from FeatureLevelSwitch}, Alpha=@n9) {constA=1}
  )
  n23 = TextureSampleParameter2D(UVs=@n2.PS) {
   parameterName=Reticle
   samplerSource=SSM_Clamp_WorldGroupSettings
   samplerType=SAMPLERTYPE_LinearColor
   texture=&/HUD_SniperCrossHairs_v1_Athena
  }
  n24 = Multiply(A=StaticSwitchParameter(True=@n23.R, False=Constant()) {defaultValue=true, parameterName=EnableReticleTexture}, B=@n9) {desc=LensOpacity}
  n25 = LinearInterpolate(A=@n22, B=VectorParameter() {parameterName=Reticle Color, defaultValue=<rgba 0 0 0 1>}.RGB, Alpha=@n24) {constB=0}
  n26 = VectorParameter() {parameterName=Reticle Color_Center, defaultValue=<rgba 0 0 0 1>}
  n27 = Constant()
  n28 = StaticSwitchParameter(
   True=LinearInterpolate(
    A=@n25
    B=Multiply(A=@n26.RGB, B=@n24)
    Alpha=StaticSwitchParameter(True=@n23.G, False=@n27) {defaultValue=true, parameterName=EnableReticleTexture}
   )
   False=@n25
  ) {parameterName=UseScopedAr}
  n29 = Multiply(A=Subtract(A=@n14.RGB) {constB=0.5}, B=Constant3Vector() {constant=<rgba 2 -2 2 0>})
  n30 = Multiply(
   A=ComponentMask(None=MaterialFunctionCall() {materialFunction=/Engine/Functions/GetPostProcessFunctions/GetAmbientCubemapTint}.AmbientCubemapTint) {b=true, g=true, r=true}
   B=TextureSampleParameterCube!(UVs=Transform(None=@n29) {transformSourceType=TRANSFORMSOURCE_View}, Apply View MipBias) {parameterName=MaterialExpressionTextureSampleParameterCube_0, texture=/Game/Packages/Fortress_Sky/TexturesHDR/T_hdrsky_test03_low}.RGB
  )
  n31 = OneMinus(None=@n14.B)
  n32 = Multiply(
   A=Multiply(A=Multiply(A=@n9, B=@n30), B=Power(Base=@n31) {constExponent=1.5})
   B=ScalarParameter() {defaultValue=4.0551, parameterName=Scope Reflection Intensity}
  )
  n33 = StaticSwitchParameter(True=Add(A=@n28, B=Multiply(A=@n6, B=@n32) {desc=mask reflection with AO}) {desc=Fake reflection}, False=@n28) {defaultValue=true, parameterName=FakeReflection}
  n34 = SphereMask(A=Constant() {r=0.5}, B=@n2.PS, Radius=ScalarParameter() {defaultValue=2, parameterName=VignetteRadius}) {attenuationRadius=2, hardnessPercent=0}
  n35 = ScalarParameter() {defaultValue=1, parameterName=BlendAmount}
  n36 = Constant2Vector() {g=0.5, r=0.5}
  n37 = SphereMask(
   A=@n2.PS
   B=@n36
   Radius=ScalarParameter() {defaultValue=0.5, parameterName=Cutoff Mask}
   Hardness=ScalarParameter() {defaultValue=1, parameterName=Cutoff Mask Hardness}
  )
  LinearInterpolate(
   A=@n22
   B=ShadingPathSwitch(Default=Multiply(A=@n33, B=@n34) {desc=Vignette}, Mobile=@n28) {desc=Migrated from FeatureLevelSwitch}
   Alpha=StaticSwitchParameter(True=Multiply(A=@n37, B=@n35), False=@n35) {parameterName=Use Cutoff Mask}
  )
  MaterialFunctionCall() {materialFunction=/Engine/Functions/GetPostProcessFunctions/GetAmbientCubemapIntensity}
  n38 = SphereMask(
   A=@n2.PS
   B=@n36
   Radius=ScalarParameter() {defaultValue=0.4615, parameterName=Cutoff Mask Zoom}
   Hardness=ScalarParameter() {defaultValue=0.8, parameterName=Cutoff Mask Hardness Zoom}
  )
  n39 = ScalarParameter() {defaultValue=1, parameterName=BlendAmount}
  NamedRerouteDeclaration(None=StaticSwitchParameter(True=Multiply(A=@n38, B=@n39), False=@n39) {parameterName=Use Cutoff Mask}) {name=cutoffmask, nodeColor=<rgba 1 0 0.6562 1>}
  SphereMask(A=@n2.PS, B=Constant() {r=0.5}, Radius=ScalarParameter() {defaultValue=0.5, parameterName=LensSize}) {hardnessPercent=99}
 ]
}
