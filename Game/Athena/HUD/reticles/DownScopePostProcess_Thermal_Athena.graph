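# Post-process material that renders the thermal vision view inside a weapon's down-sight scope, masked to a circle from AttachScope.
# ThermalScale from MPC_ThermalGlobal drives a transition that stretches and pixelates scene UVs, with mobile skipping pixelation.
# Scene red is remapped into a T_ThermalHeat gradient lookup, biased hotter for unoccluded custom-depth objects outside stencil 2.
# Stencil 123 pixels and the area outside the scope keep scene color.
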Material /Game/Athena/HUD/reticles/DownScopePostProcess_Thermal_Athena
graph {
 properties {
  bEnableResponsiveAA=true
  blendMode=BLEND_Translucent
  blendableLocation=BL_SceneColorAfterDOF
  materialDomain=MD_PostProcess
  pixelDepthOffsetMode=PDOM_Legacy
  usedShadingModels=Default Lit
 }
 nodes [
  Constant() {r=1}
  Constant() {r=0.4}
  Constant() {r=0.8}
  Constant() {r=2}
  Constant() {r=0.6}
  n1 = CollectionParameter() {collection=/Game/Athena/Items/Weapons/Abilities/Thermal/MPC_ThermalGlobal, parameterName=ThermalScale}
  n2 = Multiply(A=@n1, B=Constant() {r=1.5})
  n3 = Constant() {r=1.7}
  n4 = OneMinus(None=Clamp(None=Multiply(A=@n2, B=@n3) {constB=1.7}))
  n5 = Add(A=Multiply(A=@n4, B=Constant() {r=4}) {constB=4}, B=Constant() {r=1})
  n6 = Clamp(None=Multiply(A=@n2, B=Constant() {r=2}) {constB=1.7})
  n7 = Multiply(A=@n6, B=@n6)
  n8 = TextureCoordinate()
  n9 = Multiply(A=ViewProperty() {property=MEVP_ViewSize}.Property, B=LinearInterpolate(A=Constant() {r=0.05}, Alpha=Multiply(A=@n1, B=@n1) {constB=2}))
  n10 = Floor(None=Multiply(A=@n8, B=@n9))
  n11 = Constant() {r=100}
  n12 = Constant() {r=-100}
  n13 = MaterialFunctionCall!(RotateDialMask, Normal WS) {materialFunction=/Game/Packages/Fortress_SharedMaterials/Base_Material_Functions/AttachScope}
  n14 = Distance(A=VertexInterpolator(VS=ComponentMask(None=@n13.Post process uvs 0-1 uvs) {g=true, r=true}).PS, B=Constant() {r=0.5})
  n15 = Clamp(None=Min(A=LinearInterpolate(A=@n11, B=@n12, Alpha=@n14), B=ScalarParameter() {defaultValue=1, parameterName=BlendAmount}))
  n16 = ShadingPathSwitch(
   Default=LinearInterpolate(A=@n8, B=Divide(A=@n10, B=@n9), Alpha=@n15)
   Mobile=MaterialFunctionCall(UVs=TextureCoordinate(), Texture Scale=ScalarParameter() {defaultValue=1.17, parameterName=Mobile_ScopeMaskScale}) {materialFunction=&ScaleUVsByCenter}.UVs
  ) {desc=Migrated from FeatureLevelSwitch}
  n17 = LinearInterpolate(
   A=@n16
   B=MaterialFunctionCall(UVs=@n16, Texture Scale=AppendVector(A=@n5, B=@n7)) {materialFunction=&ScaleUVsByCenter}.UVs
   Alpha=Multiply(A=@n4, B=@n15)
  )
  n18 = ComponentMask(None=SceneTexture(UVs=@n17) {bFiltered=true, sceneTextureId=PPI_SceneDepth}.Color) {r=true}
  n19 = Add(A=ScalarParameter() {defaultValue=5, parameterName=Test}, B=@n18) {constB=0.5}
  n20 = ComponentMask(None=SceneTexture(UVs=@n17) {bFiltered=true, sceneTextureId=PPI_CustomDepth}.Color) {r=true}
  n21 = Constant() {r=1}
  If(
   A=ShadingPathSwitch(Default=@n19, Mobile=@n18) {desc=Migrated from FeatureLevelSwitch}
   B=@n20
   "A > B"=@n21
   "A == B"=@n21
   "A < B"=Constant()
  )
  n22 = ComponentMask(None=SceneTexture(UVs=@n17) {bFiltered=true, sceneTextureId=PPI_PostProcessInput0}.Color) {b=true, g=true, r=true}
  n23 = Multiply(A=ComponentMask(None=@n22) {r=true}, B=ScalarParameter() {defaultValue=1})
  n24 = ScalarParameter() {defaultValue=0.5, parameterName=SceneColorPow}
  n25 = Multiply(
   A=Clamp(None=Power(Base=@n23, Exp=@n24) {constExponent=0.75}, Max=ScalarParameter() {defaultValue=0.6, parameterName=SceneColorMax}) {maxDefault=0.35, minDefault=0.01}
   B=ScalarParameter() {defaultValue=0.7, parameterName=FinalMulti}
  )
  n26 = CollectionParameter() {collection=&MPC_ThermalGlobal, parameterName=ThermalScale}
  n27 = Clamp(None=Subtract(A=@n19, B=@n20))
  n28 = MaterialFunctionCall!(ID=Constant() {r=2}) {materialFunction=/Game/Packages/Fortress_SharedMaterials/Stencil_ID_Definitions/StencilMaskCompare}
  n29 = LinearInterpolate(
   A=@n25
   B=Add(A=@n25, B=ScalarParameter() {defaultValue=0.56, parameterName=CharacterBias}) {constB=0.65}
   Alpha=Min(A=Multiply(A=@n26, B=@n27), B=OneMinus(None=@n28.Result))
  )
  n30 = OneMinus(None=MaterialFunctionCall(ID=Constant() {r=123}) {materialFunction=&StencilMaskCompare}.Result)
  Reroute(None=LinearInterpolate(
    A=@n22
    B=TextureSample(UVs=@n29) {samplerSource=SSM_Clamp_WorldGroupSettings, texture=/Game/Athena/Items/Weapons/Abilities/Thermal/T_ThermalHeat}.RGB
    Alpha=Min(A=@n15, B=@n30)
   ))
  ScalarParameter() {defaultValue=0.05, parameterName=SampleSize}
 ]
}
