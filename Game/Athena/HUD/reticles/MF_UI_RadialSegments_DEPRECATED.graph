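# Deprecated UI material function that outputs a radial segmented mask, apparently for HUD reticles.
# Input UVs are rotated, offset to center and converted to an angle, then multiplied by a floored Repeat count.
# The result samples a segment gradient texture, whose red channel is smoothstepped using SegmentSpacing and EdgeSmooth to set gaps and edge softness.
# A sine-based alternative to the gradient lookup is left unconnected to the output.
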MaterialFunction /Game/Athena/HUD/reticles/MF_UI_RadialSegments_DEPRECATED
graph {
 properties {libraryCategoriesText [UI]}
 nodes [
  n1 = MaterialFunctionCall(
   UVs=FunctionInput(Preview=TextureCoordinate()) {bUsePreviewValueAsDefault=true, inputName=UV, inputType=FunctionInput_Vector2}
   "Rotation Angle (0-1)"=FunctionInput(Preview=Constant()) {bUsePreviewValueAsDefault=true, inputName=Rotation, inputType=FunctionInput_Scalar}
  ) {materialFunction=&CustomRotator}
  n2 = FunctionInput(Preview=Constant2Vector() {g=-0.5, r=-0.5}) {bUsePreviewValueAsDefault=true, inputName=CenterOffset, inputType=FunctionInput_Vector2}
  n3 = Constant() {r=10}
  n4 = Multiply(
   A=MaterialFunctionCall(Vector or UVs=Add(A=@n1.Rotated Values, B=@n2)) {materialFunction=&VectorToRadialValue}.Vector Converted to Angle
   B=Floor(None=FunctionInput(Preview=@n3) {bUsePreviewValueAsDefault=true, inputName=Repeat, inputType=FunctionInput_Scalar})
  )
  n5 = FunctionInput(Preview=Constant() {r=0.8}) {bUsePreviewValueAsDefault=true, inputName=SegmentSpacing, inputType=FunctionInput_Scalar}
  n6 = FunctionInput(Preview=Constant() {r=0.002}) {bUsePreviewValueAsDefault=true, inputName=EdgeSmooth, inputType=FunctionInput_Scalar}
  FunctionOutput(None=MaterialFunctionCall(
    Alpha=TextureSample(UVs=@n4) {texture=/Game/Global/Textures/Default/Masks_and_Gradients/T_UI_ReticleSegmentGradient}.R
    Min=@n5
    Max=Add(A=@n5, B=@n6)
   ) {materialFunction=&SmoothStep}.Result)
  OneMinus(None=Abs(None=Sine(None=@n4) {period=2}))
 ]
}
