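# UI material for a reticle that draws a segmented circular progress ring with fill and reload amounts.
# Two radial-gradient smoothsteps form the ring, and a radial-segment function cuts the gaps, with optional time-driven rotation.
# Angle-gradient smoothsteps blend background to fill color by PercentFillAmount and mask opacity by PercentReloadAmount, while ReloadFade scales only the unfilled alpha.
# The angle-gradient and segment functions used are marked deprecated.
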Material /Game/Athena/HUD/reticles/M_UI_Reticle_SegmentedProgress
graph {
 properties {blendMode=BLEND_Translucent, materialDomain=MD_UI, pixelDepthOffsetMode=PDOM_Legacy, usedShadingModels=Default Lit}
 nodes [
  n1 = VectorParameter() {parameterName=BGColor, defaultValue=<rgba 0.0052 0.0052 0.0056 0.6>}
  n2 = OneMinus(None=ScalarParameter() {defaultValue=0.5, parameterName=PercentFillAmount})
  n3 = MaterialFunctionCall(
   Alpha=MaterialFunctionCall!(UVs, Rotation=ScalarParameter() {defaultValue=-0.25, parameterName=FillStartRotation}) {materialFunction=/Game/UI/Foundation/Materials/MF_UI_AngleGradient_DEPRECATED}.Result
   Min=@n2
   Max=Add(A=@n2, B=ScalarParameter() {defaultValue=0.0015, parameterName=FillEdgeSmooth})
  ) {materialFunction=&SmoothStep}
  LinearInterpolate(A=@n1.RGB, B=VectorParameter() {parameterName=FillColor, defaultValue=<rgba 1 1 1 1>}.RGB, Alpha=@n3.Result)
  n4 = MaterialFunctionCall() {materialFunction=&RadialGradientExponential}
  n5 = ScalarParameter() {defaultValue=0.005, parameterName=OuterSize}
  n6 = ScalarParameter() {defaultValue=0.002, parameterName=EdgeSmooth}
  n7 = Add(A=ScalarParameter() {defaultValue=0.005, parameterName=OuterSize}, B=ScalarParameter() {defaultValue=0.1, parameterName=RingSize})
  n8 = ScalarParameter() {defaultValue=0.002, parameterName=EdgeSmooth}
  n9 = Subtract(
   A=MaterialFunctionCall(Alpha=@n4.RadialGradientExponential, Min=@n5, Max=Add(A=@n5, B=@n6)) {materialFunction=&SmoothStep}.Result
   B=MaterialFunctionCall(Alpha=@n4.RadialGradientExponential, Min=@n7, Max=Add(A=@n7, B=@n8)) {materialFunction=&SmoothStep}.Result
  )
  n10 = Multiply(A=Time(), B=ScalarParameter() {parameterName=ConstantRotation})
  n11 = Constant() {r=1}
  n12 = MaterialFunctionCall!(
   EdgeSmooth=ScalarParameter() {defaultValue=0.002, parameterName=EdgeSmooth}
   Repeat=ScalarParameter() {defaultValue=10, parameterName=Repeat}
   CenterOffset
   Rotation=Add(A=ScalarParameter() {parameterName=StartingRotation}, B=Fmod(A=@n10, B=@n11))
   UV
   SegmentSpacing=ScalarParameter() {defaultValue=0.8, parameterName=SegmentSpacing}
  ) {materialFunction=&/MF_UI_RadialSegments_DEPRECATED}
  n13 = MaterialFunctionCall(Rotation=ScalarParameter() {defaultValue=-0.25, parameterName=FillStartRotation}) {materialFunction=&MF_UI_AngleGradient_DEPRECATED}
  n14 = OneMinus(None=ScalarParameter() {defaultValue=1, parameterName=PercentReloadAmount})
  n15 = Add(A=@n14, B=ScalarParameter() {defaultValue=0.0015, parameterName=FillEdgeSmooth})
  n16 = ScalarParameter() {defaultValue=1, parameterName=ReloadFade}
  Multiply(
   A=Multiply(
    A=Subtract(A=@n9, B=@n12.Result)
    B=MaterialFunctionCall(Alpha=@n13.Result, Min=@n14, Max=@n15) {materialFunction=&SmoothStep}.Result
   )
   B=LinearInterpolate(B=Multiply(A=@n1.A, B=@n16), Alpha=OneMinus(None=@n3.Result)) {constA=1}
  )
 ]
}
