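# UI material that draws a translucent radial cooldown ring for HUD reticles.
# Rotated UVs become a radial angle, which is compared against CooldownPercent, remapped from PercentCompensation, to make a fill mask with adjustable hardness.
# That mask blends backing and fill colors and their separate alphas.
# Outer and inverted inner exponential radial gradients cut the ring shape, and FinalAlpha scales the resulting opacity.
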Material /Game/Athena/HUD/reticles/Materials/M_UI_CooldownRing
graph {
 properties {blendMode=BLEND_Translucent, materialDomain=MD_UI, pixelDepthOffsetMode=PDOM_Legacy, usedShadingModels=Default Lit}
 nodes [
  n1 = ScalarParameter() {defaultValue=0.25, group=Section, parameterName=Section_Orientation}
  n2 = TextureCoordinate()
  n3 = Multiply(A=Subtract(
    A=MaterialFunctionCall!(Rotation_Fraction=@n1, TexCoord=@n2) {materialFunction=/FortUILibrary/Foundation/Art_ToolKit/AngledBoxes/MF_UI_Rotator_Fractions}.Result
    B=Constant() {r=0.5}
   )) {constB=2}
  n4 = Subtract(
   A=Saturate(None=MaterialFunctionCall(Vector or UVs=@n3) {materialFunction=&VectorToRadialValue}.Vector Converted to Angle)
   B=LinearInterpolate(A=ScalarParameter() {parameterName=PercentCompensation}, Alpha=ScalarParameter() {defaultValue=0.333, parameterName=CooldownPercent})
  )
  n5 = ScalarParameter() {defaultValue=256, parameterName=Fill_Hardness}
  n6 = OneMinus(None=Saturate(None=Multiply(A=@n4, B=@n5)))
  LinearInterpolate(
   A=VectorParameter() {parameterName=BackingColor, defaultValue=<rgba 0 0 1 1>}.RGB
   B=VectorParameter() {parameterName=Color, defaultValue=<rgba 0 1 0 1>}.RGB
   Alpha=@n6
  )
  n7 = TextureCoordinate()
  n8 = MaterialFunctionCall(
   UVs=@n7
   Radius=ScalarParameter() {defaultValue=0.5, parameterName=OuterRadius}
   Density=ScalarParameter() {defaultValue=24, parameterName=EdgeHardness_Outer}
  ) {materialFunction=&RadialGradientExponential}
  n9 = OneMinus(None=MaterialFunctionCall(
    UVs=@n7
    Radius=ScalarParameter() {defaultValue=0.4, parameterName=InnerRadius}
    Density=ScalarParameter() {defaultValue=22, parameterName=EdgeHardness_Inner}
   ) {materialFunction=&RadialGradientExponential}.RadialGradientExponential)
  n10 = DotProduct(
   A=LinearInterpolate(
    A=ScalarParameter() {defaultValue=0.25, parameterName=Fuel_BackingAlpha}
    B=ScalarParameter() {defaultValue=1, parameterName=Fuel_FillAlpha}
    Alpha=@n6
   )
   B=Saturate(None=DotProduct(A=@n8.RadialGradientExponential, B=@n9))
  )
  Saturate(None=Multiply(A=Saturate(None=@n10), B=ScalarParameter() {defaultValue=1, parameterName=FinalAlpha}))
 ]
}
