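# Translucent UI material that draws a segmented radial charge-up progress arc, for use in HUD reticles.
# A circle outline SDF shapes the ring, and MF_UI_RadialArc computes the arc between start and end angles with its Progress fill.
# A static switch can snap progress to segment steps, using the floored segment count.
# The fill is combined by max with a dimmer stepped gradient, then cut by a sector SDF mask that spaces the segments.
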Material /Game/Athena/HUD/reticles/Materials/M_UI_Progress_ChargeUp
graph {
 properties {blendMode=BLEND_Translucent, materialDomain=MD_UI, pixelDepthOffsetMode=PDOM_Legacy, usedShadingModels=Default Lit}
 nodes [
  n1 = MaterialFunctionCall(OutlineThickness=ScalarParameter() {defaultValue=0.05, parameterName=Thickness, sliderMax=1}) {materialFunction=&MF_UI_Circle}
  n2 = ScalarParameter() {defaultValue=0.5, parameterName=Progress, sliderMax=1}
  n3 = ScalarParameter() {defaultValue=1, parameterName=Segment Count}
  n4 = MaterialFunctionCall!("Value (0-1)"=@n2, Steps=NamedRerouteDeclaration(None=Floor(None=@n3)) {name=Segments Count, nodeColor=<rgba 0 0 0 1>}) {materialFunction=/FortUILibrary/MaterialFunctions/Basics/MF_UI_StepRatio}
  n5 = MaterialFunctionCall!(
   UV=MaterialFunctionCall(
    UVs=MaterialFunctionCall() {materialFunction=&GetUserInterfaceUV}.Normalized UV
    "Rotation Angle (0-1)"=Constant() {r=0.5}
   ) {materialFunction=&CustomRotator}.Rotated Values
   "Arc Start Angle (0-1)"=ScalarParameter() {defaultValue=0.97, parameterName=Start Angle, sliderMax=1, sortPriority=31}
   "Arc End Angle (0-1)"=ScalarParameter() {defaultValue=0.2543, parameterName=End Angle, sliderMax=1}
   "Fill Ratio (0-1)"=OneMinus(None=StaticSwitchParameter(True=@n4.Stepped Ratio, False=@n2) {parameterName=Snap Progress})
   Fill Clockwise
  ) {materialFunction=/FortUILibrary/MaterialFunctions/Basics/MF_UI_RadialArc}
  n6 = SmoothStep(Value=@n5.Radial Fill) {constMax=0.05}
  n7 = NamedRerouteUsage()
  n8 = MaterialFunctionCall("Value (0-1)"=OneMinus(None=@n5.Arc Mask), Steps=@n7) {materialFunction=&MF_UI_StepRatio}
  n9 = Multiply(A=Saturate(None=Add(A=@n8.Stepped Gradient) {constB=0.25})) {constB=0.5}
  n10 = NamedRerouteDeclaration(None=OneMinus(None=ScalarParameter() {defaultValue=0.08, desc=Segment Spacing, parameterName="Segment Spacing ", sliderMax=1})) {name=Segment Spacing, nodeColor=<rgba 0 0 0 1>}
  n11 = OneMinus(None=SmoothStep(Value=MaterialFunctionCall!(UVs, Gradient Map=@n5.Arc Mask, Sector Count=@n7, Sector Spacing=@n10, Fixed Spacing) {materialFunction=/FortUILibrary/MaterialFunctions/Basics/MF_UI_Sectors}.Parallel Sectors SDF) {constMax=0.3})
  Multiply(A=OneMinus(None=SmoothStep(Value=@n1.SDFOutline) {constMax=0.03}), B=Multiply(A=Max(A=@n6, B=@n9), B=@n11))
  NamedRerouteDeclaration(None=@n5.Arc Mask) {nodeColor=<rgba 0 0.4609 1 1>}
  NamedRerouteDeclaration(None=@n11) {name=Mask, nodeColor=<rgba 0 0 0 1>}
  StaticBoolParameter() {parameterName=Fixed Spacing}
  StaticBoolParameter() {parameterName=Clockwise}
  VectorParameter() {parameterName=ReticleColor, defaultValue=<rgba 1 1 1 1>}
 ]
}
