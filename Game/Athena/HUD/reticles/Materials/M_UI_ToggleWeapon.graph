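# Translucent UI material for the weapon-toggle reticle, drawing an inner dot and an outer ring, each with a soft shadow.
# Two MF_UI_Circle calls supply signed distance fields, the outer on center-scaled UVs, which SmoothStep turns into masks published as named reroutes.
# Mask sums are weighted by per-circle alpha parameters, with size, ring thickness, blur, shadow spread, shadow opacity and dot color also parameterized.
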Material /Game/Athena/HUD/reticles/Materials/M_UI_ToggleWeapon
graph {
 properties {blendMode=BLEND_Translucent, materialDomain=MD_UI, pixelDepthOffsetMode=PDOM_Legacy, usedShadingModels=Default Lit}
 nodes [
  n1 = Multiply(A=NamedRerouteUsage(), B=ScalarParameter() {defaultValue=0.5, parameterName=Shadow_Opacity})
  n2 = NamedRerouteUsage()
  n3 = NamedRerouteUsage()
  n4 = NamedRerouteUsage()
  Add(
   A=Multiply(A=ScalarParameter() {defaultValue=1, group=OuterCircle, parameterName=OuterCircle_Alpha}, B=Add(A=@n1, B=@n2))
   B=Multiply(A=ScalarParameter() {defaultValue=1, group=Inner Circle, parameterName=InnerCircle_Alpha}, B=Add(A=@n3, B=@n4))
  )
  LinearInterpolate(A=NamedRerouteUsage(), Alpha=NamedRerouteUsage())
  n5 = MaterialFunctionCall(
   UVs=TextureCoordinate()
   Size=ScalarParameter() {defaultValue=0.55, group=Inner Circle, parameterName=Inner_Size}
   OutlineThickness=Constant()
  ) {materialFunction=&MF_UI_Circle}
  NamedRerouteDeclaration(None=SmoothStep(Value=@n5.SDF) {constMax=0, constMin=0.08}) {name=InnerCircle, nodeColor=<rgba 0 0.9301 1 1>}
  NamedRerouteDeclaration(None=SmoothStep(Value=@n5.SDFOutline) {constMax=0, constMin=0.05}) {name=InnerCircleShadow, nodeColor=<rgba 0 0.9301 1 1>}
  n6 = ScalarParameter() {defaultValue=0.05, parameterName=OuterRing_Blur}
  n7 = MaterialFunctionCall(
   UVs=MaterialFunctionCall(Texture Scale=Constant() {r=0.88}) {materialFunction=&ScaleUVsByCenter}.UVs
   Size=ScalarParameter() {defaultValue=0.9, group=OuterCircle, parameterName=Outer_Size}
   OutlineThickness=ScalarParameter() {defaultValue=0.07, group=OuterCircle, parameterName=Outer_Thickness}
  ) {materialFunction=&MF_UI_Circle}
  NamedRerouteDeclaration(None=OneMinus(None=SmoothStep(Max=@n6, Value=@n7.SDFOutline) {constMax=0.05})) {name=OuterCircle, nodeColor=<rgba 1 0 0.2747 1>}
  n8 = ScalarParameter() {defaultValue=0.15, parameterName=Shadow_Spread}
  NamedRerouteDeclaration(None=OneMinus(None=SmoothStep(Max=@n8, Value=@n7.SDFOutline) {constMax=0.1})) {name=OuterCircleShadow, nodeColor=<rgba 1 0 0.2747 1>}
  VectorParameter() {parameterName=DotColor, defaultValue=<rgba 0.0495 0 0.0007 1>}
 ]
}
